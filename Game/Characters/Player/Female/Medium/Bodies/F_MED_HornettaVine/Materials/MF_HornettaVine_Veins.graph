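# Material function for the HornettaVine player character that produces pulsing deep veins and a noisy radial glow, plus pulse and UV outputs.
# A derivative-built tangent-space camera vector drives BumpOffset, giving top, half-offset mid and bottom parallax UVs.
# Two vein texture channels are shaped by time-offset sine pulses and multiplied; Low quality drops the mid layer and glow noise.
# The glow mask reuses PositionX and Radius, ignoring PositionY and Hardness; "Noise Only" outputs a constant.
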MaterialFunction /Game/Characters/Player/Female/Medium/Bodies/F_MED_HornettaVine/Materials/MF_HornettaVine_Veins
graph {nodes [
  Composite() {subgraphName=CameraVectorTS}
  n1 = NamedRerouteUsage()
  n2 = ScalarParameter() {defaultValue=1.5, group=HornettaVine, parameterName=UVTiling_Deep}
  n3 = Time()
  n4 = ScalarParameter() {defaultValue=0.4, group=HornettaVine, parameterName=Veins_PulseSpeed}
  n5 = ScalarParameter() {defaultValue=0.5, group=HornettaVine, parameterName=Veins_PulseMin}
  n6 = ScalarParameter() {defaultValue=1, group=HornettaVine, parameterName=Veins_PulseMax}
  n7 = MaterialFunctionCall!(Sine Phase=Multiply(A=@n3, B=@n4), Value 1=@n5, Value 2=@n6) {materialFunction=/Engine/Functions/Engine_MaterialFunctions02/Utility/Sine_Remapped}
  n8 = Power(
   Base=TextureSample(UVs=Multiply(A=@n1, B=@n2)) {samplerSource=SSM_Wrap_WorldGroupSettings, texture=/Game/Global/Textures/Default/Tileable_Patterns/T_Veins_Blurred}.R
   Exp=@n7.Remapped Sine
  )
  n9 = NamedRerouteUsage()
  n10 = ScalarParameter() {defaultValue=8, group=HornettaVine, parameterName=UVTiling_Mid}
  n11 = ScalarParameter() {defaultValue=0.5, group=HornettaVine, parameterName=Veins_PulseOffset}
  n12 = MaterialFunctionCall(Sine Phase=Multiply(A=@n4, B=Add(A=@n3, B=@n11)) {constB=0.4}, Value 1=@n5, Value 2=@n6) {materialFunction=&Sine_Remapped}
  n13 = Power(
   Base=TextureSample(UVs=Multiply(A=@n9, B=@n10)) {samplerSource=SSM_Wrap_WorldGroupSettings, texture=&T_Veins_Blurred}.G
   Exp=@n12.Remapped Sine
  )
  FunctionOutput(None=QualitySwitch(Default=Multiply(A=@n8, B=@n13), Low=@n8)) {outputName=Deep Veins}
  FunctionOutput(None=QualitySwitch(Default=@n12.Remapped Sine, Low=Constant() {r=1})) {outputName=Pulse_Mid, sortPriority=20}
  FunctionOutput(None=NamedRerouteUsage()) {outputName=UVs_Top, sortPriority=10}
  FunctionOutput(None=NamedRerouteUsage()) {outputName=UVs_Mid, sortPriority=11}
  FunctionOutput(None=NamedRerouteUsage()) {outputName=UVs_Bot, sortPriority=12}
  FunctionOutput(None=@n7.Remapped Sine) {outputName=Pulse_Bot, sortPriority=20}
  n14 = NamedRerouteUsage()
  n15 = VectorParameter() {group=HornettaVine, parameterName=RadialGradientParams, channelNames {a=Hardness, b=Radius, g=PositionY, r=PositionX}}
  n16 = SphereMask(A=@n14, B=AppendVector(A=@n15.PositionX, B=@n15.PositionX), Radius=@n15.Radius, Hardness=@n15.Radius)
  n17 = Multiply(A=@n14, B=ScalarParameter() {defaultValue=8, group=HornettaVine, parameterName=RadialGrandient_NoiseTiling})
  n18 = Constant() {r=0.4}
  n19 = QualitySwitch(
   Default=Power(
    Base=TextureSample(UVs=@n17) {texture=&AtmosphericCloudNoise01}.RGB
    Exp=ScalarParameter() {defaultValue=1.5, group=HornettaVine, parameterName=RadialGrandient_NoisePower}
   )
   Low=@n18
  )
  FunctionOutput(None=Multiply(A=Multiply(A=@n16, B=@n19), B=ScalarParameter() {defaultValue=1.5, group=HornettaVine, parameterName=RadialGrandient_Intensity})) {outputName=Glow Gradient With Noise, sortPriority=1}
  FunctionOutput(None=@n18) {outputName=Glow Gradient Noise Only, sortPriority=2}
  n20 = TextureCoordinate()
  NamedRerouteDeclaration(None=@n20) {name=UVs_Top, nodeColor=<rgba 0.1953 1 0 1>}
  n21 = WorldPosition() {worldPositionShaderOffset=WPT_ExcludeAllShaderOffsets}
  n22 = PixelNormalWS()
  n23 = CrossProduct(A=DDY(Value=@n21.XYZ), B=@n22)
  n24 = TextureCoordinate()
  n25 = MaterialFunctionCall(Float2=DDX(Value=@n24)) {materialFunction=&BreakOutFloat2Components}
  n26 = CrossProduct(A=@n22, B=DDX(Value=@n21.XYZ))
  n27 = MaterialFunctionCall(Float2=DDY(Value=@n24)) {materialFunction=&BreakOutFloat2Components}
  n28 = Add(A=Multiply(A=@n23, B=@n25.R), B=Multiply(A=@n26, B=@n27.R))
  n29 = DotProduct(A=@n28, B=@n28)
  n30 = Add(A=Multiply(A=@n23, B=@n25.G), B=Multiply(A=@n26, B=@n27.G))
  n31 = DotProduct(A=@n30, B=@n30)
  n32 = Divide(B=SquareRoot(None=Max(A=@n29, B=@n31)))
  n33 = MaterialFunctionCall!(
   VectorToTransform=CameraVectorWS()
   BasisX=Multiply(A=@n28, B=@n32) {desc=Adjusted T}
   BasisY=Multiply(A=@n32, B=@n30) {desc=Adjusted T}
   BasisZ=@n22
  ) {materialFunction=/Engine/Functions/Engine_MaterialFunctions02/Math/InverseTransformMatrix}
  MaterialExpressionReroute_0 = Reroute(None=@n33.Transformed Vector 3)
  n34 = MaterialFunctionCall!(
   HeightTexture
   TextureCoordinates=@n20
   ReferencePlane=Constant() {r=0.5}
   HeightRatioInput=ScalarParameter() {defaultValue=0.05, group=HornettaVine, parameterName=Veins_Depth}
   CameraVector=@MaterialExpressionReroute_0.OutVector
  ) {materialFunction=/Engine/Functions/Engine_MaterialFunctions02/Utility/BumpOffset_Advanced.BumpOffset_advanced}
  NamedRerouteDeclaration(None=@n34.UVs) {name=UVs_Bot, nodeColor=<rgba 1 0 0.0469 1>}
  NamedRerouteDeclaration(None=Add(A=@n34.Undeformed UVs, B=Multiply(A=@n34.Distortion) {constB=0.5})) {name=UVs_Mid, nodeColor=<rgba 0 1 1 1>}
  PinBase!(OutVector=@n33.Transformed Vector 3) {reroutePins [OutVector {expression=&self:MaterialExpressionReroute_0}]}
  PinBase!()
 ]}
