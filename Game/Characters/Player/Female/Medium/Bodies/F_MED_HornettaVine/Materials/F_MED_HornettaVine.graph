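# Character body material for HornettaVine that adds pulsing, glowing veins between the shared CharacterParent PreFX and PostFX functions.
# MF_HornettaVine_Veins supplies vein, glow and pulse masks that shape the parent's emissive, with UV-based shoulder masks switching to a pulsing, view-dependent glow.
# A sphere mask on X-mirrored pre-skinned position extends the arm gradient.
# Time-of-day functions scale glow, desaturate it at night and control optional base color tinting.
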Material /Game/Characters/Player/Female/Medium/Bodies/F_MED_HornettaVine/Materials/F_MED_HornettaVine
graph {
 properties {
  bUseMaterialAttributes=true
  bUsedWithMeshDeformer=true
  bUsedWithMorphTargets=true
  bUsedWithNanite=true
  bUsedWithSkeletalMesh=true
  pixelDepthOffsetMode=PDOM_Legacy
  shadingModel=MSM_FromMaterialExpression
  usedShadingModels=Default Lit | Subsurface | Clear Coat | Subsurface Profile | Cloth
 }
 nodes [
  n1 = MaterialFunctionCall() {materialFunction=&/MF_HornettaVine_Veins}
  Composite!(InputPin=@n1.UVs_Mid) {subgraphName=ShoulderMask}
  Composite!() {subgraphName=ShoulderSeamGradient}
  n2 = MaterialFunctionCall() {materialFunction=&MF_CharacterParent_2023_PreFX}
  n3 = ScalarParameter() {defaultValue=1, group=HornettaVine, parameterName=TOD_MorningEvening}
  n4 = MaterialFunctionCall(
   Value_Morning=@n3
   Value_Day=Constant() {r=1}
   Value_Evening=@n3
   Value_Night=ScalarParameter() {defaultValue=0.7, group=HornettaVine, parameterName=TOD_Night}
  ) {materialFunction=&Scalar_Time_of_Day}
  n5 = TextureCoordinate()
  n6 = ComponentMask(None=@n5) {r=true}
  n7 = ComponentMask(None=@n5) {g=true}
  n8 = Power(Base=Abs(None=ConstantBiasScale(None=@n7) {bias=-0.5, scale=2})) {constExponent=8}
  n9 = Multiply(
   A=Ceil(None=Subtract(A=@n6) {constB=0.76})
   B=MaterialFunctionCall(In=OneMinus(None=@n8), Contrast=Constant() {r=1}) {materialFunction=&CheapContrast}.Result
  )
  MaterialExpressionReroute_9 = Reroute(None=@n9)
  n10 = Constant2Vector() {g=0.038, r=0.057}
  MaterialExpressionReroute_8 = Reroute(None=@n1.UVs_Mid)
  n11 = Constant2Vector() {g=-0.415, r=-0.19}
  n12 = Add(
   A=MaterialFunctionCall(UVs=Add(A=@n10, B=@MaterialExpressionReroute_8.InputPin), Texture Scale=Constant() {r=0.172}) {materialFunction=&ScaleUVsByCenter}.0-1 mask
   B=MaterialFunctionCall(UVs=Add(A=@MaterialExpressionReroute_8.InputPin, B=@n11), Texture Scale=Constant() {r=0.18}) {materialFunction=&ScaleUVsByCenter}.0-1 mask
  )
  MaterialExpressionReroute_7 = Reroute(None=@n12)
  n13 = Subtract(A=@MaterialExpressionReroute_9.Output, B=@MaterialExpressionReroute_7.Output)
  n14 = MaterialFunctionCall(Float3=NamedRerouteUsage()) {materialFunction=&BreakOutFloat3Components}
  n15 = MaterialFunctionCall(X=Abs(None=@n14.R), Y=@n14.G, Z=@n14.G) {materialFunction=&MakeFloat3}
  n16 = Constant3Vector() {constant=<rgba 13.76 -5.1 135.33 0>}
  n17 = ScalarParameter() {defaultValue=40, group=HornettaVine, parameterName=ArmsGrandient_Radius}
  n18 = ScalarParameter() {group=HornettaVine, parameterName=ArmsGrandient_Hardness}
  n19 = Multiply(
   A=Multiply(A=@n1.Glow Gradient Noise Only, B=SphereMask(A=@n15.Result, B=@n16, Radius=@n17, Hardness=@n18) {attenuationRadius=25, hardnessPercent=1})
   B=ScalarParameter() {defaultValue=1.5, group=HornettaVine, parameterName=ArmsGrandient_Intensity}
  )
  n20 = Multiply(A=Power(Base=@n4.Result) {constExponent=4}, B=Max(A=@n1.Glow Gradient With Noise, B=Multiply(A=@n13, B=@n19)))
  n21 = GetMaterialAttributes(None=@n2.Result) {attributeGetTypes [#1, #2]}
  n22 = ComponentMask(None=@n2.Tech Art Mask) {r=true}
  n23 = ScalarParameter() {defaultValue=0.05, group=HornettaVine, parameterName=Veins_Surface_Translucency}
  n24 = Multiply(A=@n21.Emissive Color, B=Min(A=Max(A=@n22, B=@n23) {constB=0.2}, B=@n1.Deep Veins))
  n25 = ScalarParameter() {defaultValue=1, group=HornettaVine, parameterName=ShoulderIntensity}
  n26 = Power(
   Base=DotProduct(A=CameraVectorWS(), B=VertexNormalWS())
   Exp=ScalarParameter() {defaultValue=1.5, group=HornettaVine, parameterName=ShoulderFresnelPower}
  )
  n27 = NamedRerouteUsage()
  n28 = ScalarParameter() {defaultValue=0.6, group=HornettaVine, parameterName=ShoulderFresnelPulsePower}
  n29 = Multiply(A=@n21.Emissive Color, B=Multiply(A=Multiply(A=@n25, B=@n26), B=Power(Base=@n27, Exp=@n28)))
  n30 = Multiply(A=@n4.Result, B=Multiply(A=Add(A=@n20), B=LinearInterpolate(A=@n24, B=@n29, Alpha=@MaterialExpressionReroute_7.Output)))
  n31 = Constant()
  n32 = MaterialFunctionCall(
   Value_Morning=@n31
   Value_Day=@n31
   Value_Evening=@n31
   Value_Night=ScalarParameter() {defaultValue=0.1, group=HornettaVine, parameterName=TOD_Night_Desaturation}
  ) {materialFunction=&Scalar_Time_of_Day}
  n33 = MaterialFunctionCall(In=Power(Base=Desaturation(None=@n21.Emissive Color)) {constExponent=0.5}, Contrast=Constant() {r=0.1}) {materialFunction=&CheapContrast}
  n34 = Multiply(A=@n21.Base Color, B=LinearInterpolate(B=@n24, Alpha=Power(Base=@n33.Result) {constExponent=0.5}) {constA=1})
  n35 = ScalarParameter() {group=HornettaVine, parameterName=AffectBaseColor}
  n36 = MaterialFunctionCall(
   Value_Morning=@n35
   Value_Day=@n35
   Value_Evening=@n35
   Value_Night=ScalarParameter() {group=HornettaVine, parameterName=TOD_Night_AffectsBaseColor}
  ) {materialFunction=&Scalar_Time_of_Day}
  MaterialFunctionCall(
   Material=SetMaterialAttributes!(
    MaterialAttributes=@n2.Result
    Emissive Color=Desaturation(None=@n30, Fraction=@n32.Result)
    Base Color=LinearInterpolate(A=@n21.Base Color, B=@n34, Alpha=@n36.Result)
   ) {attributeSetTypes [#1, #2]}
   M=@n2.M
   Pre Skinned Local Position=@n2.M
   Distance Blend=@n2.M
   Fresnel=@n2.M
  ) {materialFunction=&MF_CharacterParent_2023_PostFX}
  NamedRerouteDeclaration(None=@n1.Pulse_Bot) {name=Pulse, nodeColor=<rgba 0.1719 1 0 1>}
  NamedRerouteDeclaration(None=@n2.Pre skinned Local Position) {name=PreSkinPos, nodeColor=<rgba 1 0.0234 0 1>}
  PinBase() {reroutePins [InputPin {expression=&self:MaterialExpressionReroute_8}]}
  PinBase!(Output=@n12) {reroutePins [Output {expression=&self:MaterialExpressionReroute_7}]}
  PinBase!()
  PinBase!(Output=@n9) {reroutePins [Output {expression=&self:MaterialExpressionReroute_9}]}
 ]
}
