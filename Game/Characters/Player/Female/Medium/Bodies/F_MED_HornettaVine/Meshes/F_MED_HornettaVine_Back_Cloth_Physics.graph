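# Physics asset for the back cloth piece on the medium female HornettaVine player body, defining a short dyn_cloth_a bone chain.
# A small sphere at the base bone anchors three capsules (parent, 01, 02) linked in series by constraints.
# Swing limits tighten from base to tip, while twist stays equally narrow at every joint.
# A large unconstrained capsule on the right thigh gives the leg a collision shape alongside the cloth.
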PhysicsAsset /Game/Characters/Player/Female/Medium/Bodies/F_MED_HornettaVine/Meshes/F_MED_HornettaVine_Back_Cloth_Physics
bodies {
 dyn_cloth_a_01 [{shapeType=Capsule, length=5.764, radius=2.1125, shapeName=dyn_cloth_a_01_capsule, center=<4.2956 0 0>, rotation=<rot 90 0 0>}]
 dyn_cloth_a_02 [{shapeType=Capsule, length=5.5946, radius=1.7913, shapeName=dyn_cloth_a_02_capsule, center=<3.3647 0 0>, rotation=<rot 90 0 0>}]
 dyn_cloth_a_base [{shapeType=Sphere, radius=0.5, shapeName=dyn_cloth_a_base_sphere}]
 dyn_cloth_a_parent [{shapeType=Capsule, length=3.9922, radius=2.1443, shapeName=dyn_cloth_a_parent_capsule, center=<3.7512 0 0>, rotation=<rot 90 0 0>}]
 thigh_r [{shapeType=Capsule, length=32.3189, radius=11.3414, shapeName=thigh_r_capsule, center=<18.377 0 0>, rotation=<rot 90 0 0>}]
}
constraints [
 {bone1Name=dyn_cloth_a_parent, bone2Name=dyn_cloth_a_base, swing1LimitDegrees=20, swing2LimitDegrees=20, twistLimitDegrees=3}
 {bone1Name=dyn_cloth_a_01, bone2Name=dyn_cloth_a_parent, swing1LimitDegrees=15, swing2LimitDegrees=15, twistLimitDegrees=3}
 {bone1Name=dyn_cloth_a_02, bone2Name=dyn_cloth_a_01, swing1LimitDegrees=10, swing2LimitDegrees=10, twistLimitDegrees=3}
]
