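# Physics asset for the side cloth of the HornettaVine medium female player body, simulating three dyn_cloth bone chains (b, c, d).
# Each chain runs capsules from a small base sphere through constraints with narrow swing and tight twist limits.
# An angularly free cross-constraint ties c_02 to d_02, linking those neighboring strands.
# A large capsule on the right thigh twist bone supplies a leg collision shape for the cloth.
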PhysicsAsset /Game/Characters/Player/Female/Medium/Bodies/F_MED_HornettaVine/Meshes/F_MED_HornettaVine_Side_Cloth_Physics
bodies {
 dyn_cloth_b_01 [{shapeType=Capsule, length=4.212, radius=1.951, shapeName=dyn_cloth_b_01_capsule, center=<3.142 0 0>, rotation=<rot 90 0 0>}]
 dyn_cloth_b_base [{shapeType=Sphere, radius=0.5, shapeName=dyn_cloth_b_base_sphere}]
 dyn_cloth_b_parent [{shapeType=Capsule, length=1.9126, radius=1.682, shapeName=dyn_cloth_b_parent_capsule, center=<1.9789 0 0>, rotation=<rot 90 0 0>}]
 dyn_cloth_c_01 [{shapeType=Capsule, length=6.3162, radius=1.5376, shapeName=dyn_cloth_c_01_capsule, center=<3.6848 0.5923 0>, rotation=<rot 90 0 -9.0389>}]
 dyn_cloth_c_02 [{shapeType=Capsule, length=3.1445, radius=1.3402, shapeName=dyn_cloth_c_02_capsule, center=<2.0231 0 0>, rotation=<rot 90 0 0>}]
 dyn_cloth_c_base [{shapeType=Sphere, radius=0.5, shapeName=dyn_cloth_c_base_sphere}]
 dyn_cloth_c_parent [{shapeType=Capsule, length=2.1762, radius=1.5587, shapeName=dyn_cloth_c_parent_capsule, center=<2.507 0 0>, rotation=<rot 90 0 0>}]
 dyn_cloth_d_01 [{shapeType=Capsule, length=4.6989, radius=2.1602, shapeName=dyn_cloth_d_01_capsule, center=<3.5802 0 0>, rotation=<rot 90 0 0>}]
 dyn_cloth_d_02 [{shapeType=Capsule, length=4.335, radius=2.3502, shapeName=dyn_cloth_d_02_capsule, center=<2.8695 0 0>, rotation=<rot 90 0 0>}]
 dyn_cloth_d_03 [{shapeType=Capsule, length=5.868, radius=1.6766, shapeName=dyn_cloth_d_03_capsule, center=<4.1286 0 0>, rotation=<rot 90 0 0>}]
 dyn_cloth_d_base [{shapeType=Sphere, radius=0.5, shapeName=dyn_cloth_d_base_sphere}]
 dyn_cloth_d_parent [{shapeType=Capsule, length=1.7536, radius=1.9148, shapeName=dyn_cloth_d_parent_capsule, center=<2.5419 0 0>, rotation=<rot 90 0 0>}]
 thigh_twist_01_r [{
   shapeType=Capsule
   length=23.8114
   radius=6.1886
   shapeName=thigh_twist_01_r_capsule
   center=<-5.6517 0 -0.1019>
   rotation=<rot 88.5432 179.9996 179.9997>
  }]
}
constraints [
 {bone1Name=dyn_cloth_b_01, bone2Name=dyn_cloth_b_parent, swing1LimitDegrees=10, swing2LimitDegrees=10, twistLimitDegrees=3}
 {bone1Name=dyn_cloth_b_parent, bone2Name=dyn_cloth_b_base, swing1LimitDegrees=10, swing2LimitDegrees=10, twistLimitDegrees=3}
 {bone1Name=dyn_cloth_d_03, bone2Name=dyn_cloth_d_02, swing1LimitDegrees=20, swing2LimitDegrees=20, twistLimitDegrees=3}
 {bone1Name=dyn_cloth_d_02, bone2Name=dyn_cloth_d_01, swing1LimitDegrees=20, swing2LimitDegrees=20, twistLimitDegrees=3}
 {bone1Name=dyn_cloth_d_01, bone2Name=dyn_cloth_d_parent, swing1LimitDegrees=20, swing2LimitDegrees=20, twistLimitDegrees=3}
 {bone1Name=dyn_cloth_d_parent, bone2Name=dyn_cloth_d_base, swing1LimitDegrees=20, swing2LimitDegrees=20, twistLimitDegrees=3}
 {bone1Name=dyn_cloth_c_02, bone2Name=dyn_cloth_c_01, swing1LimitDegrees=10, swing2LimitDegrees=20, twistLimitDegrees=3}
 {bone1Name=dyn_cloth_c_01, bone2Name=dyn_cloth_c_parent, swing1LimitDegrees=10, swing2LimitDegrees=20, twistLimitDegrees=3}
 {bone1Name=dyn_cloth_c_parent, bone2Name=dyn_cloth_c_base, swing1LimitDegrees=10, swing2LimitDegrees=20, twistLimitDegrees=3}
 {
  bone1Name=dyn_cloth_c_02
  bone2Name=dyn_cloth_d_02
  swing1LimitDegrees=45
  swing1Motion=Free
  swing2LimitDegrees=45
  swing2Motion=Free
  twistLimitDegrees=45
  twistMotion=Free
 }
]
